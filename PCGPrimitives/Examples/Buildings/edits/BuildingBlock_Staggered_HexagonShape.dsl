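; PCG example graph that builds a staggered, three-tier building block from spline shapes, hexagonal per the asset name.
; Each tier assigns a shape grammar, spawns its points as assets and offsets footprints, which feed planar meshes and contract into the next tier.
; The top is contracted, filled uniformly and randomly thinned before spawning assets.
; A separate base shape gets filled with assets and is exposed as Out for other graphs.
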(graph BuildingBlock_Staggered_HexagonShape
  (bind Subgraph_0 (Subgraph :Graph /PCGPrimitives/Primitives/Create/Create_Spline_Shape.Create_Spline_Shape)) ; The base square shape MUST be connected to the\nDefaultOutputNode Out pin of the graph for \nother graph to get and use it in their logic.
  (bind Subgraph_11 (Subgraph :Graph /PCGPrimitives/Primitives/Create/Create_Spline_Shape.Create_Spline_Shape))
  (bind Subgraph_5 (Subgraph :Graph /PCGPrimitives/Primitives/Assign/Assign_ShapeGrammarDefinition.Assign_ShapeGrammarDefinition))
  (bind Subgraph_19 (Subgraph :Graph /PCGPrimitives/Primitives/Transform/Transform_Offset.Transform_Offset))
  (bind Subgraph_3 (Subgraph :Graph /PCGPrimitives/Primitives/Transform/Transform_Areas_Contract.Transform_Areas_Contract))
  (bind Subgraph_15 (Subgraph :Graph /PCGPrimitives/Primitives/Assign/Assign_ShapeGrammarDefinition.Assign_ShapeGrammarDefinition))
  (bind Subgraph_29 (Subgraph :Graph /PCGPrimitives/Primitives/Transform/Transform_Offset.Transform_Offset))
  (bind Subgraph_38 (Subgraph :Graph /PCGPrimitives/Primitives/Transform/Transform_Areas_Contract.Transform_Areas_Contract))
  (bind Subgraph_18 (Subgraph :Graph /PCGPrimitives/Primitives/Assign/Assign_ShapeGrammarDefinition.Assign_ShapeGrammarDefinition))
  (bind Subgraph_30 (Subgraph :Graph /PCGPrimitives/Primitives/Transform/Transform_Offset.Transform_Offset))
  (bind Subgraph_2 (Subgraph :Graph /PCGPrimitives/Primitives/Create/Create_Mesh_Planar.Create_Mesh_Planar))
  (bind Subgraph_16 (Subgraph :Graph /PCGPrimitives/Primitives/Fill/Fill_Areas_Uniformly.Fill_Areas_Uniformly))
  (bind Subgraph_20 (Subgraph :Graph /PCGPrimitives/Primitives/Assign/Assign_Assets.Assign_Assets))
  (bind Subgraph_17 (Subgraph :Graph /PCGPrimitives/Primitives/Spawn/Spawn_Assets.Spawn_Assets))
  (bind Subgraph_26 (Subgraph :Graph /PCGPrimitives/Primitives/Spawn/Spawn_Assets.Spawn_Assets))
  (bind Subgraph_27 (Subgraph :Graph /PCGPrimitives/Primitives/Transform/Transform_Areas_Contract.Transform_Areas_Contract))
  (bind Subgraph_1 (Subgraph :Graph /PCGPrimitives/Primitives/Fill/Fill_Areas_Uniformly.Fill_Areas_Uniformly))
  (bind Subgraph_4 (Subgraph :Graph /PCGPrimitives/Primitives/Filter/Filter_RandomChoice.Filter_RandomChoice))
  (bind Subgraph_6 (Subgraph :Graph /PCGPrimitives/Primitives/Spawn/Spawn_Assets.Spawn_Assets))
  (connect Subgraph_27 Out Subgraph_1 PrimaryInput)
  (connect Subgraph_19 Out Subgraph_2 PrimaryInput)
  (connect Subgraph_29 Out Subgraph_2 PrimaryInput)
  (connect Subgraph_30 Out Subgraph_2 PrimaryInput)
  (connect Subgraph_19 Out Subgraph_3 PrimaryInput)
  (connect Subgraph_1 Out Subgraph_4 PrimaryInput)
  (connect Subgraph_11 Out Subgraph_5 PrimaryInput)
  (connect Subgraph_4 Selected Subgraph_6 PrimaryInput)
  (connect Subgraph_3 Out Subgraph_15 PrimaryInput)
  (connect Subgraph_0 Out Subgraph_16 PrimaryInput)
  (connect Subgraph_20 Out Subgraph_17 PrimaryInput)
  (connect Subgraph_38 Out Subgraph_18 PrimaryInput)
  (connect Subgraph_5 Footprints Subgraph_19 PrimaryInput)
  (connect Subgraph_16 Out Subgraph_20 PrimaryInput)
  (connect Subgraph_5 Points Subgraph_26 PrimaryInput)
  (connect Subgraph_15 Points Subgraph_26 PrimaryInput)
  (connect Subgraph_18 Points Subgraph_26 PrimaryInput)
  (connect Subgraph_30 Out Subgraph_27 PrimaryInput)
  (connect Subgraph_15 Footprints Subgraph_29 PrimaryInput)
  (connect Subgraph_18 Footprints Subgraph_30 PrimaryInput)
  (connect Subgraph_29 Out Subgraph_30 PrimaryInput)
  (connect Subgraph_29 Out Subgraph_38 PrimaryInput)
  (output Out Subgraph_0))
